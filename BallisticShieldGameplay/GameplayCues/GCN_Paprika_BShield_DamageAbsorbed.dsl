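; Burst gameplay cue notify for damage absorbed by the ballistic shield, driving its impact sounds and particles.
; On burst it plays a BShielded hit-notify event through the source pawn's SoundLibrary component, letting experiences control hitmarker audio.
; It spawns an absorb sound parameterized by stagger, melee source tags and whether the effect causer is locally viewed.
; It also feeds cue parameters into the particle ImpactStage, then calls the parent handler.
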(fn FireAbsorbDamageSFX
  (params (My%20Target :type Object%20Reference) (GameplayCueParameters :type Gameplay%20Cue%20Parameters%20Structure%20%28by%20ref%29))
  (bind _gameplaycueparameters (Variables|GetGameplayCueParameters))
  (bind (_normalizedmagnitude _rawmagnitude _effectcontext _matchedtagname _originaltag _aggregatedsourcetags _aggregatedtargettags _location) (Ability|GameplayCue|BreakGameplayCueParameters _gameplaycueparameters))
  (bind (_normalizedmagnitude_2 _rawmagnitude_2 _effectcontext_2 _matchedtagname_2 _originaltag_2 _aggregatedsourcetags_2 _aggregatedtargettags_2 _location_2 _normal _instigator _effectcauser) (Ability|GameplayCue|BreakGameplayCueParameters (Variables|GetGameplayCueParameters)))
  ;[0 0.04611 0.1042] Hit Sweetener
  (bind _returnvalue (Audio|SpawnSoundatLocation My%20Target "/BallisticShieldGameplay/Sounds/BallisticShield_AbsorbDamage_Cue.BallisticShield_AbsorbDamage_Cue" _location))
  (Audio|Parameter|SetFloatParameter _returnvalue "Stagger" (Ability|GameplayCue|BreakGameplayCueParameters (Variables|GetGameplayCueParameters)) :!class /Script/AudioExtensions.AudioParameterControllerInterface)
  (Audio|Parameter|SetBooleanParameter _returnvalue "isMeleeHit" (GameplayTags|HasAnyTags _aggregatedsourcetags (GameplayTags|MakeLiteralGameplayTagContainer "(GameplayTags=((TagName=\"Item.Weapon.Melee\"),(TagName=\"Weapon.Melee\")))")) :!class /Script/AudioExtensions.AudioParameterControllerInterface)
  (bind _asfort_pawn (Utilities|Casting|CastToFortPawn _effectcauser))
  (Audio|Parameter|SetBooleanParameter _returnvalue "isLocallyViewed" (Pawn|IsLocallyViewed _asfort_pawn) :!class /Script/AudioExtensions.AudioParameterControllerInterface))

(fn FireHitmarkerSFX
  (params (Parameters :type Gameplay%20Cue%20Parameters%20Structure%20%28by%20ref%29))
  (bind (_normalizedmagnitude _rawmagnitude _effectcontext _matchedtagname _originaltag _aggregatedsourcetags _aggregatedtargettags _location _normal _instigator _effectcauser _sourceobject) (Ability|GameplayCue|BreakGameplayCueParameters Parameters))
  ;[0 0 0] New way of doing things with SoundLibrary, allows Experiences to specify if they want sounds to play
  (bind _asfort_pawn (Utilities|Casting|CastToFortPawn _sourceobject))
  (bind _asfort_hit_notify_sound_library_component (Utilities|Casting|CastToFortHitNotifySoundLibraryComponent (SoundLibrary|GetSoundLibraryComponent _asfort_pawn "(TagName=\"SoundLibrary.HitNotify\")" :!class /Script/FortniteGame.FortPawn)))
  (Utilities|IsValid (SoundLibrary|GetSoundLibraryComponent _asfort_pawn "(TagName=\"SoundLibrary.HitNotify\")" :!class /Script/FortniteGame.FortPawn)
    (:Is%20Valid
      (SoundLibrary|ConfigureContext _asfort_hit_notify_sound_library_component (Utilities|Struct|MakeFortHitNotifyContextSettings _effectcauser _asfort_pawn) :!class /Script/FortniteGame.FortHitNotifySoundLibraryComponent)
      (SoundLibrary|PlayEvent _asfort_hit_notify_sound_library_component "(TagName=\"SoundLibrary.HitNotify.Event.BShielded\")"))))

(fn OnBurst(EitherVFX-preferthisversion)
  (params (MyTarget :type Actor%20Object%20Reference) (Parameters :type Gameplay%20Cue%20Parameters%20Structure%20%28by%20ref%29) (ParticleComponents :type Array%20of%20FXSystem%20Component%20Object%20References) (AudioComponents :type Array%20of%20Audio%20Component%20Object%20References) (BurstCameraShakeInstance :type Camera%20Shake%20Base%20Object%20Reference) (BurstDecalInstance :type Decal%20Actor%20Object%20Reference))
  (Utilities|FlowControl|Sequence
    (:then_0
      (CallFunction|FireHitmarkerSFX Parameters)
      (CallFunction|FireAbsorbDamageSFX MyTarget Parameters)
      (Effects|Components|ParticleSystem|SetFloatParameter (Utilities|Array|Get(acopy) ParticleComponents) "ImpactStage" (Ability|GameplayCue|BreakGameplayCueParameters Parameters)))
    (:then_1
      (|CallParentFunction MyTarget Parameters ParticleComponents AudioComponents BurstCameraShakeInstance BurstDecalInstance)
      (return))))
